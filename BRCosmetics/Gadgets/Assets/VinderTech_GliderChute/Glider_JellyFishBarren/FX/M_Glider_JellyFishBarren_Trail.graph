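# Translucent, unlit, two-sided Niagara ribbon material for the JellyFishBarren glider trail, tinted by particle color through an emissive adjustment function.
# Panning Gaussian noise distorts center-scaled UVs, whose distance from a center point drives a sine for procedural radial waves.
# Particle alpha, a sphere mask, camera depth fade and two generated bands shape those waves into soft-edged opacity.
# A shading path switch adds depth fade on default and deferred paths but skips it on mobile.
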Material /BRCosmetics/Gadgets/Assets/VinderTech_GliderChute/Glider_JellyFishBarren/FX/M_Glider_JellyFishBarren_Trail
graph {
 properties {
  bAutomaticallySetUsageInEditor=false
  bUsedWithNiagaraRibbons=true
  blendMode=BLEND_Translucent
  pixelDepthOffsetMode=PDOM_Legacy
  shadingModel=MSM_Unlit
  twoSided=true
  usedShadingModels=Unlit
 }
 nodes [
  n1 = ParticleColor()
  MaterialFunctionCall(Emissive=@n1.RGB) {materialFunction=&MF_FX_EmissiveAdjustments}
  n2 = MaterialFunctionCall(
   UVs=TextureCoordinate()
   Texture Scale=AppendVector(
    A=ScalarParameter() {defaultValue=1, parameterName=Scale X, sliderMax=5}
    B=ScalarParameter() {defaultValue=1, parameterName=Scale Y, sliderMax=5}
   )
  ) {materialFunction=&ScaleUVsByCenter}
  n3 = Panner(
   Coordinate=TextureCoordinate()
   Speed=AppendVector(
    A=ScalarParameter() {defaultValue=-1, parameterName=Speed X, sliderMax=5, sliderMin=-5}
    B=ScalarParameter() {parameterName=Speed Y, sliderMax=5, sliderMin=-5}
   )
  )
  n4 = Multiply(
   A=Subtract(A=TextureSample(UVs=@n3) {texture=/Game/Effects/Fort_Effects/Textures/Tile/T_GaussianNoise_01_M}.R) {constB=0.5}
   B=ScalarParameter() {defaultValue=0.1, parameterName=UV Distortion, sliderMax=2}
  )
  n5 = ScalarParameter() {defaultValue=0.5, parameterName=Sphere Center X}
  n6 = ScalarParameter() {defaultValue=0.5, parameterName=Sphere Center Y}
  n7 = Multiply(
   A=Distance(A=Add(A=@n2.UVs, B=@n4), B=AppendVector(A=@n5, B=@n6))
   B=ScalarParameter() {defaultValue=1, parameterName=Frequency, sliderMax=10}
  )
  n8 = ScalarParameter() {defaultValue=0.5, parameterName=Sphere Mask Center X}
  n9 = ScalarParameter() {defaultValue=0.5, parameterName=Sphere Mask Center Y}
  n10 = Multiply(
   A=Multiply(A=@n1.A, B=Sine(None=@n7))
   B=SphereMask(
    A=TextureCoordinate()
    B=AppendVector(A=@n8, B=@n9)
    Radius=ScalarParameter() {defaultValue=0.5, parameterName=sphere mask Radius}
    Hardness=ScalarParameter() {parameterName=Sphere mask Hardness}
   ) {attenuationRadius=0.5, hardnessPercent=0}
  )
  n11 = MaterialFunctionCall(Fade Length=ScalarParameter() {parameterName="Cam Fade Length (S)"}, Fade Offset=ScalarParameter() {parameterName="Cam Fade Offset (S)"}) {materialFunction=&CameraDepthFade}
  n12 = Constant() {r=0.5}
  n13 = ScalarParameter() {defaultValue=0.1, parameterName=Band sharpness}
  n14 = StaticBool() {value=true}
  n15 = Multiply(
   A=Multiply(
    A=Multiply(A=@n10, B=@n11.Result)
    B=MaterialFunctionCall!(Width=@n12, Sharpness=@n13, Offset, Direction Switch=@n14, Compare, Input Coordinates) {materialFunction=/Engine/Functions/Engine_MaterialFunctions03/Procedurals/GeneratedBand}.Result
   )
   B=MaterialFunctionCall(Width=Constant() {r=0.5}, Sharpness=ScalarParameter() {defaultValue=0.75, parameterName=Band sharpness Sides}) {materialFunction=&GeneratedBand}.Result
  )
  n16 = DepthFade(Opacity=@n15, FadeDistance=ScalarParameter() {defaultValue=10, parameterName=FadeDistance, sliderMax=100})
  ShadingPathSwitch(Default=@n16, Deferred=@n16, Mobile=@n15)
 ]
}
