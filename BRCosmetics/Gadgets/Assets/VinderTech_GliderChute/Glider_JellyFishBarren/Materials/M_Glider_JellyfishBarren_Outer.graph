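# Translucent lit outer-shell material for the Jellyfish Barren glider, with Fresnel-driven opacity, color and blurred scene color.
# Emissive combines thin-film color, parallax FX glow warped by water normals, faked sun specular and cubemap reflection.
# Panning pre-skinned-position projections, blended between two planes, supply distortion noise and water normals layered over the baked normal.
# Noise-modulated sine sway, vertex-masked and scaled by SwayPower and animation-read TentacleVertexShader, drives world position offset.
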Material /BRCosmetics/Gadgets/Assets/VinderTech_GliderChute/Glider_JellyFishBarren/Materials/M_Glider_JellyfishBarren_Outer
graph {
 properties {
  bContactShadows=true
  bScreenSpaceReflections=true
  bUseMaterialAttributes=true
  bUsedWithMeshDeformer=true
  bUsedWithSkeletalMesh=true
  blendMode=BLEND_Translucent
  pixelDepthOffsetMode=PDOM_Legacy
  translucencyLightingMode=TLM_VolumetricDirectional
  usedShadingModels=Default Lit
 }
 nodes [
  n1 = Multiply(A=VectorParameter() {parameterName=Sway Direction, defaultValue=<rgba 0 0 0 1>}.RGB, B=ScalarParameter() {parameterName=Sway Intensity})
  n2 = TextureSample() {
   samplerType=SAMPLERTYPE_LinearGrayscale
   texture=/BRCosmetics/Gadgets/Assets/VinderTech_GliderChute/Glider_Jellyfish/Textures/Glider_Jellyfish_VertexMask
  }
  n3 = Constant3Vector() {constant=<rgba 0 0.2 0 0>}
  n4 = Multiply(A=Time(), B=ScalarParameter() {defaultValue=0.33, parameterName=Sway Variation Rate})
  n5 = Sine(None=@n4)
  n6 = Sine(None=@n4) {period=1.3}
  n7 = Add(A=@n3, B=AppendVector(A=AppendVector(A=@n5, B=@n6), B=Constant()))
  n8 = ScalarParameter() {defaultValue=0.1, parameterName=Sway Variance Amount}
  n9 = PreSkinnedPosition()
  n10 = Divide(A=@n9, B=ScalarParameter() {defaultValue=40, group=PanningEmissive, parameterName=Triplanner_ConstantScaleDivisor})
  n11 = Time()
  n12 = VectorParameter() {group=PanningEmissive, parameterName=TriPlanner Speed, defaultValue=<rgba 0 0 0.2 1>}
  n13 = Add(A=Multiply(A=ScalarParameter() {defaultValue=15, parameterName=Distortion Tiling}, B=@n10), B=Frac(None=Multiply(A=@n11, B=@n12.RGB)))
  n14 = TextureObjectParameter() {group=PanningEmissive, parameterName=EmissiveGradient, texture=/Game/Global/Textures/Default/Noise/T_Tiling_Noise_01}
  n15 = LinearInterpolate(
   A=TextureSample(UVs=ComponentMask(None=@n13) {b=true, r=true}, Tex=@n14) {samplerType=SAMPLERTYPE_Masks, texture=/Game/Weapons/WeaponSkins/Textures/T_Swirly_Noise_highContrast}.RGB
   B=TextureSample(UVs=ComponentMask(None=@n13) {b=true, g=true}, Tex=@n14) {samplerType=SAMPLERTYPE_Masks, texture=&T_Swirly_Noise_highContrast}.RGB
  )
  n16 = Multiply(A=Normalize(VectorInput=LinearInterpolate(A=@n3, B=@n7, Alpha=@n8)), B=@n15)
  Add(A=Multiply(A=Multiply(A=@n1) {constB=0.5}, B=@n2.R), B=@n16)
  Constant() {r=1}
  n17 = VectorParameter() {parameterName=Base Shell Color, defaultValue=<rgba 1 0.4479 0.4479 1>}
  n18 = TextureSampleParameter2D() {
   parameterName=Baked Normal
   samplerType=SAMPLERTYPE_Normal
   texture=/BRCosmetics/Gadgets/Assets/VinderTech_GliderChute/Glider_Jellyfish/Textures/T_Glider_Jellyfish_N
  }
  n19 = Fresnel(ExponentIn=ScalarParameter() {defaultValue=0.8131, parameterName=TranparencyFresnel}, Normal=Transform(None=@n18.RGB))
  n20 = Saturate(None=OneMinus(None=@n19))
  n21 = Saturate(None=Power(Base=@n20, Exp=ScalarParameter() {defaultValue=1.2, parameterName=BlurFresnel}) {constExponent=4})
  LinearInterpolate(A=@n17.RGB, Alpha=@n21)
  n22 = CollectionParameter() {
   collection=&FortniteMaterialParameters
   parameterName=SunAndMoonModelDirectionalVector
  }
  n23 = MaterialFunctionCall(Float3=@n22) {materialFunction=&BreakOutFloat3Components}
  n24 = MaterialFunctionCall(X=@n23.R, Y=@n23.R, Z=Constant() {r=1}) {materialFunction=&MakeFloat3}
  n25 = Multiply(
   A=DotProduct(A=ComponentMask(None=@n24.Result) {b=true, g=true, r=true}, B=ReflectionVectorWS())
   B=ScalarParameter() {defaultValue=0.8993, parameterName=SpecMult}
  )
  n26 = ScalarParameter() {defaultValue=70, parameterName=Spec Highlight Power}
  n27 = Multiply(A=Clamp(None=Power(Base=@n25, Exp=@n26) {constExponent=20}), B=ScalarParameter() {defaultValue=3, parameterName=Spec Intense})
  n28 = TextureSampleParameterCube!(UVs=ReflectionVectorWS() {bNormalizeCustomWorldNormal=true}, Bias=Constant() {r=3}, Apply View MipBias) {
   mipValueMode=TMVM_MipBias
   parameterName=CubeMap
   samplerSource=SSM_Wrap_WorldGroupSettings
   samplerType=SAMPLERTYPE_LinearColor
   texture=/Game/Packages/Fortress_Sky/TexturesHDR/T_AthenaSkylight
  }
  n29 = VectorParameter() {parameterName=Cube Map Multiplier, defaultValue=<rgba 0.22 0.22 0.5 0>}
  n30 = MaterialFunctionCall() {materialFunction=/Engine/Functions/GetPostProcessFunctions/GetAmbientCubemapTint}
  n31 = Desaturation(
   None=Multiply(A=Multiply(A=@n28.RGB, B=@n29.RGB) {constB=1.5}, B=ComponentMask(None=@n30.AmbientCubemapTint) {b=true, g=true, r=true})
   Fraction=Constant() {r=0.5}
  )
  n32 = VectorParameter() {parameterName=DeepSea Emissive, defaultValue=<rgba 0 6.178 12 1>}
  n33 = TextureCoordinate()
  n34 = LinearInterpolate(A=ScalarParameter() {defaultValue=-0.5, parameterName=Jellyfish Height}, Alpha=Saturate(None=@n2.R)) {constB=0}
  n35 = ScalarParameter() {defaultValue=0.1, parameterName=WaterTiling}
  n36 = VertexInterpolator(VS=@n10)
  n37 = Multiply(A=Time(), B=VectorParameter() {group=PanningEmissive, parameterName=Water Speed, defaultValue=<rgba 0 0 0.1 1>}.RGB)
  n38 = Add(
   A=Add(A=Multiply(A=@n35, B=@n36.PS), B=Frac(None=@n37))
   B=Multiply(A=VertexInterpolator(VS=@n15).PS, B=ScalarParameter() {defaultValue=0.05, parameterName=Offset Intensity}) {constB=0.01}
  )
  n39 = TextureObjectParameter() {
   group=PanningEmissive
   parameterName=WaterNormal
   samplerType=SAMPLERTYPE_Normal
   texture=/Engine/Functions/Engine_MaterialFunctions02/ExampleContent/Textures/water_n
  }
  n40 = PreSkinnedNormal()
  n41 = MaterialFunctionCall(
   In=Abs(None=ComponentMask(None=@n40) {b=true})
   Contrast=ScalarParameter() {defaultValue=0.2, group=PanningEmissive, parameterName=TriPlanner_AxisBlendStrength}
  ) {materialFunction=&CheapContrast}
  n42 = LinearInterpolate(
   A=TextureSample(UVs=ComponentMask(None=@n38) {b=true, r=true}, Tex=@n39) {samplerType=SAMPLERTYPE_Normal, texture=&T_Swirly_Noise_highContrast}.RGB
   B=TextureSample(UVs=ComponentMask(None=@n38) {g=true, r=true}, Tex=@n39) {samplerType=SAMPLERTYPE_Normal, texture=&T_Swirly_Noise_highContrast}.RGB
   Alpha=VertexInterpolator(VS=ComponentMask(None=@n41.Result) {b=true}).PS
  )
  n43 = Multiply(A=ComponentMask(None=@n42) {g=true, r=true}, B=ScalarParameter() {defaultValue=0.01, parameterName=Emissive Warp})
  n44 = TextureSample(UVs=Add(A=BumpOffset(Coordinate=@n33, Height=@n34), B=@n43)) {
   samplerType=SAMPLERTYPE_Masks
   texture=/BRCosmetics/Gadgets/Assets/VinderTech_GliderChute/Glider_JellyFishBarren/Textures/T_JellyfishBarren_Glider_FX
  }
  n45 = Add(A=Multiply(A=Multiply(A=@n32.RGB, B=@n21), B=@n44.R), B=Add(A=@n17.RGB, B=Add(A=@n27, B=@n31)))
  n46 = ComponentMask(None=@n42) {g=true}
  n47 = Add(A=Multiply(A=LinearInterpolate(Alpha=@n46) {constA=1.5, constB=3}, B=@n19) {constA=2}, B=@n43)
  n48 = VertexInterpolator(VS=@n15)
  n49 = Multiply(
   A=Multiply(A=TextureSample(UVs=@n47) {texture=/Game/Global/Textures/Default/Thinfilm/T_ThinFilm_Spectrum_COLOR}.RGB, B=Multiply(A=@n48.PS, B=@n20))
   B=ScalarParameter() {defaultValue=1, parameterName=ThinFilm Intensity}
  )
  n50 = Constant() {r=1}
  n51 = ScalarParameter() {parameterName=SwayPower}
  n52 = Multiply(A=@n16, B=@n2.R)
  n53 = ScalarParameter() {defaultValue=0.6, parameterName=WaterIntensity}
  n54 = SetMaterialAttributes!(
   MaterialAttributes
   Base Color=LinearInterpolate(
    A=VectorParameter() {parameterName=Base Color, defaultValue=<rgba 0.0248 0.0233 0.0312 1>}.RGB
    B=VectorParameter() {parameterName=Inner Shell Color, defaultValue=<rgba 0.7166 0.9192 1 1>}.RGB
    Alpha=@n21
   )
   Emissive Color=LinearInterpolate(
    A=Add(A=@n49, B=@n45)
    B=MaterialFunctionCall!(Mask=@n50) {materialFunction=/BRCosmetics/Characters/Player/Common/SharedMaterials/SceneColor_Blur}.Result
    Alpha=@n21
   )
   Opacity=LinearInterpolate(
    A=ScalarParameter() {defaultValue=0.7333, parameterName=Opacity Min}
    B=ScalarParameter() {defaultValue=1, parameterName=Opacity Max}
    Alpha=@n19
   )
   Normal=MaterialFunctionCall(
    BaseNormal=@n18.RGB
    AdditionalNormal=MaterialFunctionCall(Normal=@n42, Flatness=@n53) {materialFunction=&FlattenNormal}.Result
   ) {materialFunction=&BlendAngleCorrectedNormals}.Result
   World Position Offset=Multiply(A=Multiply(A=@n51, B=@n52), B=ScalarParameter() {defaultValue=1, desc=This value is read from the anims, parameterName=TentacleVertexShader})
  ) {attributeSetTypes [
    #1
    #2
    #3
    #4
    #5
   ]}
  MaterialFunctionCall(In=@n54) {materialFunction=&MF_TechArtOverrides_accessories}
  ScalarParameter() {defaultValue=0.181, parameterName=Roughness}
  ScalarParameter() {defaultValue=0.75, parameterName=Specular}
  VertexInterpolator(VS=@n9)
  VertexInterpolator(VS=@n40)
 ]
}
